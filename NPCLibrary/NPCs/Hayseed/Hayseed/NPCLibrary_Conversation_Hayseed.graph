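# Conversation database for the Hayseed NPC, providing his greeting dialogue and routing players into shared NPC conversations.
# The entry point leads to a speech node with a default line about his cousin Steel.
# Requirement-gated lines, weighted by priority curve rows, react to time since last match, a first meeting, Bars held, or the Corny outfit.
# The speech then branches through link nodes to the shared NPCLibrary quest and basic services entries.
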ConversationDatabase /NPCLibrary/NPCs/Hayseed/Hayseed/NPCLibrary_Conversation_Hayseed
entry_points [{tag=Conversation.Entry.Hayseed, destinations [#1]}]
nodes [
 ConversationEntryPointNode_7 {
  id=#1
  class=ConversationEntryPointNode
  outputs [#2]
  properties {entryTag {tagName=Conversation.Entry.Hayseed}}
 }
 Speech_C_4 {
  id=#2
  class=Speech_C
  outputs [#3, #4]
  properties {
   nodeName=""
   generalConfig {
    defaultMessage=You seen my cousin Steel anywhere? I been lookin' for him.
    contextualMessages [
     {
      message="Hey there Hoss, it's been a spell! Good to see ya."
      contextRequirements [{requirement {class=FortControllerRequirement_CheckTimeSinceLastPlayedMatch}}]
      priority {curve {rowName=Tandem.Speech.Priority.WhileSincePlayed}}
     }
     {
      message="Howdy! I love meetin' a new friendly face."
      contextRequirements [{requirement {class=FortControllerRequirement_IsFirstEverConversationWithNPC}}]
      priority {curve {rowName=Tandem.Speech.Priority.FirstEverInteraction}}
     }
     {
      message="<b>Whoo-ee!</> Y'all have got a lotta Bars to spend at the farmer's market!"
      contextRequirements [{requirement {class=FortControllerRequirement_HasBars, properties {thresholdTestConfig {requirement=GreaterThanOrEqual}}}}]
      priority {curve {rowName=Tandem.Speech.Priority.HighTotalBars}}
     }
     {
      message="You know you're not supposed to be outside!"
      requirementMatchPolicy=RequireAny
      contextRequirements [{requirement {
         class=FortControllerRequirement_HasCID
         properties {allowedCIds [/BRCosmetics/Athena/Items/Cosmetics/Characters/CID_959_Athena_Commando_M_Corny]}
        }}]
      priority {curve {rowName=Tandem.Speech.Priority.OutfitReaction}}
     }
    ]
   }
  }
 }
 ConversationLinkNode_0 {id=#4, class=ConversationLinkNode, properties {remoteEntryTag {tagName=Conversation.Entry.BasicServices}}}
 ConversationLinkNode_1 {id=#3, class=ConversationLinkNode, properties {remoteEntryTag {tagName=Conversation.Entry.NPCLibrary.Quests}}}
]
